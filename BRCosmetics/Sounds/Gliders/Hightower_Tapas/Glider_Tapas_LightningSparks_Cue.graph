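# Sound cue for lightning spark one-shots on the Hightower Tapas glider, routed to the Fort_Gliders sound class.
# A Random node picks with equal weight among five big and five small spark waves.
# A Modulator varies pitch and volume on each play, over a low base volume and slightly lowered pitch.
# It uses its own attenuation asset, and a concurrency override caps instances at three, stopping the oldest.
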SoundCue /BRCosmetics/Sounds/Gliders/Hightower_Tapas/Glider_Tapas_LightningSparks_Cue
root=Modulator_0
nodes [
 {id=Modulator_0, children [Random_1], properties {pitchMax=1.2, pitchMin=0.8, volumeMax=1.2, volumeMin=0.8}}
 {
  id=Random_1
  children [
   WavePlayer_6
   WavePlayer_7
   WavePlayer_8
   WavePlayer_9
   WavePlayer_10
   WavePlayer_11
   WavePlayer_12
   WavePlayer_13
   WavePlayer_14
   WavePlayer_15
  ]
  properties {weights [1,1,1,1,1,1,1,1,1,1]}
 }
 {id=WavePlayer_10, properties {soundWaveAssetPtr=&/Glider_Tapas_SparksBig_05}}
 {id=WavePlayer_11, properties {soundWaveAssetPtr=&/Glider_Tapas_SparksSmall_01}}
 {id=WavePlayer_12, properties {soundWaveAssetPtr=&/Glider_Tapas_SparksSmall_02}}
 {id=WavePlayer_13, properties {soundWaveAssetPtr=&/Glider_Tapas_SparksSmall_03}}
 {id=WavePlayer_14, properties {soundWaveAssetPtr=&/Glider_Tapas_SparksSmall_04}}
 {id=WavePlayer_15, properties {soundWaveAssetPtr=&/Glider_Tapas_SparksSmall_05}}
 {id=WavePlayer_6, properties {soundWaveAssetPtr=&/Glider_Tapas_SparksBig_01}}
 {id=WavePlayer_7, properties {soundWaveAssetPtr=&/Glider_Tapas_SparksBig_02}}
 {id=WavePlayer_8, properties {soundWaveAssetPtr=&/Glider_Tapas_SparksBig_03}}
 {id=WavePlayer_9, properties {soundWaveAssetPtr=&/Glider_Tapas_SparksBig_04}}
]
properties {
 attenuationSettings=&/Glider_Tapas_LightningSparks_Att
 bOverrideConcurrency=true
 maxDistance=800
 pitchMultiplier=0.9
 soundClassObject=/Game/Sounds/FortSoundClassesAndModes/Fort_Gliders
 volumeMultiplier=0.12
 concurrencyOverrides {maxCount=3, resolutionRule=StopOldest, volumeScale=0.7}
}
